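# Sound cue for the Shockwave Mace's shared melee "charged" sound, built from three Charge_Start_V2 wave variants.
# An equally weighted Random node picks one variant and feeds it through a Modulator to the output.
# It uses the melee weapon sound class with small-toy creative attenuation, and a concurrency override limits it to one playing instance.
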SoundCue /ShockwaveMace/Audio/Sounds/Shared/Melee_Shared_Charge_Charged_Cue
root=Modulator_0
nodes [
 {id=Modulator_0, children [Random_2]}
 {id=Random_2, children [WavePlayer_6,WavePlayer_7,WavePlayer_8], properties {weights [1,1,1]}}
 {id=WavePlayer_6, properties {soundWaveAssetPtr=&/Melee_Shared_Charge_Start_V2_01}}
 {id=WavePlayer_7, properties {soundWaveAssetPtr=&/Melee_Shared_Charge_Start_V2_02}}
 {id=WavePlayer_8, properties {soundWaveAssetPtr=&/Melee_Shared_Charge_Start_V2_03}}
]
properties {
 attenuationSettings=/Game/Sounds/Creative/Attenuations/Fort_Creative_SmallToy_Att
 bOverrideConcurrency=true
 maxDistance=3000
 soundClassObject=&Fort_WeaponMelee
 volumeMultiplier=0.8
 concurrencyOverrides {maxCount=1}
}
