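# Sound cue for the charge-release sound shared by Shockwave Mace melee attacks.
# It randomly picks one of three equally weighted wave variants and passes it through a modulator.
# It uses the melee weapon sound class, a small-toy creative attenuation and reduced volume.
# A concurrency override allows only one instance to play at a time.
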SoundCue /ShockwaveMace/Audio/Sounds/Shared/Melee_Shared_Charge_Release_Cue
root=Modulator_1
nodes [
 {id=Modulator_1, children [Random_0]}
 {id=Random_0, children [WavePlayer_1,WavePlayer_2,WavePlayer_3], properties {weights [1,1,1]}}
 {id=WavePlayer_1, properties {soundWaveAssetPtr=&/Melee_Shared_Charge_Release_V2_01}}
 {id=WavePlayer_2, properties {soundWaveAssetPtr=&/Melee_Shared_Charge_Release_V2_02}}
 {id=WavePlayer_3, properties {soundWaveAssetPtr=&/Melee_Shared_Charge_Release_V2_03}}
]
properties {
 attenuationSettings=/Game/Sounds/Creative/Attenuations/Fort_Creative_SmallToy_Att
 bOverrideConcurrency=true
 maxDistance=3000
 soundClassObject=&Fort_WeaponMelee
 volumeMultiplier=0.6
 concurrencyOverrides {maxCount=1}
}
